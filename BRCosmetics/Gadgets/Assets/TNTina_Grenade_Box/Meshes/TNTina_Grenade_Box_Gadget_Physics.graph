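# Physics asset for the TNTina Grenade Box gadget's skeletal mesh, giving it simplified collision and joint setup.
# It has two bodies, one on the Anim bone and one on the lid bone, each a single capsule of nearly identical size.
# One constraint joins the lid to the Anim body, with swing and twist limit angles all set to 45 degrees.
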PhysicsAsset /BRCosmetics/Gadgets/Assets/TNTina_Grenade_Box/Meshes/TNTina_Grenade_Box_Gadget_Physics
bodies {
 Anim [{shapeType=Capsule, length=69.9608, radius=50.9885, center=<0 0.0014 32.3621>, rotation=<rot 0 -179.4678 -89.993>}]
 lid [{shapeType=Capsule, length=69.6738, radius=50.3979, center=<44.8701 -0.0079 -2.5608>, rotation=<rot 0 -0.2892 -89.8331>}]
}
constraints [{bone1Name=lid, bone2Name=Anim, swing1LimitDegrees=45, swing2LimitDegrees=45, twistLimitDegrees=45}]
